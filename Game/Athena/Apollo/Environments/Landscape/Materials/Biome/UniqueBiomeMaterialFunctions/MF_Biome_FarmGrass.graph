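# Farm grass biome material function that recolors an incoming grass material's BaseColor and passes the grass heightmap through as Height.
# The color blends Grasslands and Farm tints by the Apollo terrain Desert mask, then divides by a mask-driven lerp of two divisors.
# Under virtual texturing, inverse RVT output level adds a near tint, mask-blended with a third color.
# Several noise, darkening and distance-blend chains do not reach the outputs.
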MaterialFunction /Game/Athena/Apollo/Environments/Landscape/Materials/Biome/UniqueBiomeMaterialFunctions/MF_Biome_FarmGrass
graph {nodes [
  n1 = NamedRerouteUsage()
  n2 = LinearInterpolate(A=Constant() {r=0.55}, B=Constant() {r=3}, Alpha=@n1)
  Add(A=Multiply(A=Constant() {r=0.05}, B=@n1), B=Multiply(A=Saturate(None=@n2)))
  Constant3Vector() {constant=<rgba 0.0354 0.1305 0.5 1>}
  Constant3Vector() {constant=<rgba 0.65 0.4992 0.0467 1>}
  Constant() {r=0.45}
  Constant() {r=4.5}
  Constant() {r=2.5}
  n3 = GetMaterialAttributes(None=FunctionInput() {inputName=GrassMaterial, inputType=FunctionInput_MaterialAttributes}) {attributeGetTypes [#1]}
  n4 = LinearInterpolate(
   A=VectorParameter() {parameterName=GrasslandsGrassColorB, defaultValue=<rgba 0.35 0.3467 0.035 1>}.RGB
   B=VectorParameter() {parameterName=FarmGrassColorB, defaultValue=<rgba 0.0354 0.1305 0.5 1>}.RGB
   Alpha=MaterialFunctionCall!(UseLandscapeCoords) {materialFunction=/Game/Athena/Apollo/Environments/Landscape/Materials/Biome/CommonMaterials/MF_ApolloTerrainMasks}.Desert
  )
  n5 = VectorParameter() {desc=Near, parameterName=FarmGrassColorC, defaultValue=<rgba 0.0945 0.15 0.015 1>}
  n6 = ViewProperty() {property=MEVP_RuntimeVirtualTextureOutputLevel}
  n7 = Saturate(None=Multiply(
    A=ScalarParameter() {defaultValue=1.5, group=Biome_Farmgrass, parameterName=RVT_DistNearFarmGrass}
    B=ComponentMask(None=@n6.InvProperty) {r=true}
   ))
  n8 = VirtualTextureFeatureSwitch!(
   No=@n4
   Yes=LinearInterpolate(
    A=VectorParameter() {parameterName=FarmGrassColorA, defaultValue=<rgba 0.0824 0.14 0 1>}.RGB
    B=LinearInterpolate(A=@n4, B=@n5.RGB, Alpha=@n7)
    Alpha=NamedRerouteUsage()
   )
  )
  n9 = LinearInterpolate(
   A=ScalarParameter() {defaultValue=3, group=Biome_Farmgrass, parameterName=FarmGrass_DivA}
   B=ScalarParameter() {defaultValue=0.8, group=Biome_Farmgrass, parameterName=FarmGrass_DivB}
   Alpha=NamedRerouteUsage()
  )
  FunctionOutput(None=SetMaterialAttributes!(MaterialAttributes=@n3.MaterialAttributes, BaseColor=Divide(A=@n8, B=@n9)) {attributeSetTypes [#1]}) {outputName=Material}
  FunctionOutput(None=NamedRerouteDeclaration(None=FunctionInput() {inputName=GrassHeightmap, inputType=FunctionInput_Scalar}) {name=GrassHeightmap, nodeColor=<rgba 1 0 1 1>}) {outputName=Height, sortPriority=1}
  n10 = NamedRerouteUsage()
  n11 = NamedRerouteUsage()
  n12 = ScalarParameter() {defaultValue=2.5, parameterName=FarmGrassHPowD}
  n13 = Saturate(None=Add(A=Power(Base=@n11, Exp=@n12), B=ScalarParameter() {parameterName=FarmGrassHAddD}))
  n14 = Constant()
  n15 = ViewProperty() {property=MEVP_RuntimeVirtualTextureOutputLevel}
  n16 = Saturate(None=Multiply(A=ScalarParameter() {defaultValue=4, parameterName=RVT_DistNearFarmGrass-New11}, B=ComponentMask(None=@n15.InvProperty) {r=true}))
  n17 = Saturate(None=MaterialFunctionCall(
    Input=VirtualTextureFeatureSwitch(No=@n14, Yes=@n16)
    Input Low=ScalarParameter() {defaultValue=0.5, parameterName=vet8hrtb34}
    Input High=ScalarParameter() {defaultValue=1, parameterName=hhhhrt54}
    Target Low=Constant()
    Target High=Constant() {r=1}
   ) {materialFunction=&RemapValueRange}.Result)
  LinearInterpolate(
   A=LinearInterpolate(A=LinearInterpolate(A=@n10, B=@n4, Alpha=@n13), B=@n10, Alpha=@n17) {desc="A=Far"}
   B=@n17
   Alpha=ScalarParameter() {parameterName=garehw5ey3634}
  )
  NamedRerouteDeclaration(None=FunctionInput() {inputName=GrassBaseColorDetail, inputType=FunctionInput_Scalar}) {name=GrassBaseColorDetail, nodeColor=<rgba 1 0 1 1>}
  NamedRerouteDeclaration(None=@n3.BaseColor) {name=InputDiffuse, nodeColor=<rgba 1 0 1 1>}
  NamedRerouteDeclaration(None=FunctionInput() {inputName=GrassNoiseMask, inputType=FunctionInput_Scalar}) {name=GrassNoiseMask, nodeColor=<rgba 1 0 1 1>}
  n18 = NamedRerouteUsage()
  n19 = Multiply(A=ScalarParameter() {defaultValue=1, parameterName=FarmGrassDarkFar}, B=@n18)
  n20 = Multiply(A=@n18, B=ScalarParameter() {defaultValue=1, parameterName=FarmGrassDarkNear})
  n21 = Constant() {r=2.5}
  n22 = ComponentMask(None=ViewProperty() {property=MEVP_RuntimeVirtualTextureOutputLevel}.InvProperty) {r=true}
  n23 = VirtualTextureFeatureSwitch(No=Constant(), Yes=Saturate(None=Multiply(A=@n21, B=@n22)))
  n24 = Add(
   A=Multiply(A=NamedRerouteUsage(), B=ScalarParameter() {defaultValue=-2, parameterName=FarmGrassNoise_FarMod})
   B=ScalarParameter() {defaultValue=1.4, parameterName=FarmGrassNoise_FarAdd}
  )
  n25 = Power(Base=NamedRerouteUsage(), Exp=ScalarParameter() {defaultValue=2, parameterName=FarmGrassNoiseMaskthing5})
  n26 = Multiply(A=ScalarParameter() {defaultValue=4, parameterName=RVT_DistNearFarmGrass2}, B=@n22)
  n27 = LinearInterpolate(A=@n24, B=Max(A=OneMinus(None=@n25), B=@n23), Alpha=VirtualTextureFeatureSwitch(No=Constant(), Yes=Saturate(None=@n26))) {desc="Far=A"}
  Reroute(None=LinearInterpolate(A=@n4, B=LinearInterpolate(A=@n19, B=@n20, Alpha=@n23), Alpha=Saturate(None=@n27)) {desc="Far=A"})
  n28 = NamedRerouteUsage()
  n29 = Constant() {r=0.01}
  Saturate(None=Add(A=Multiply(B=@n28), B=Multiply(A=@n28, B=@n29)))
  Saturate(None=@n24)
 ]}
