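# Landscape material for the Atlantis biome, driving terrain shading, grass spawning and physical surface types.
# Common layers height-blend with forest floor (BiomeB) and farm grass (BiomeC), BiomeD alters color and roughness, then slope-based cliffs are added.
# The result feeds runtime virtual texturing, with a simpler mobile path, before a final Curie function.
# Layer and cliff masks derive physical materials; layer masks and world-height bands place grass, flowers, beach and ocean foliage.
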Material /Game/Athena/Apollo/Environments/Landscape/Materials/Biome/LandscapeBiomeMaterials/M_Biome_Atlantis
graph {
 properties {bTangentSpaceNormal=false, bUseMaterialAttributes=true, pixelDepthOffsetMode=PDOM_Legacy, usedShadingModels=Default Lit}
 nodes [
  Constant() {r=-4400}
  n1 = Reroute()
  n2 = Reroute()
  n3 = Constant3Vector() {constant=<rgba 0 1 0 0>}
  n4 = If(A=Constant(), B=@n2, "A > B"=@n3, "A == B"=@n3, "A < B"=Constant3Vector() {constant=<rgba 1 0 0 0>})
  n5 = If(A=@n2, B=@n1, "A > B"=@n4, "A == B"=@n4, "A < B"=Constant3Vector() {constant=<rgba 0 0 1 0>})
  If(A=@n1, B=Reroute(), "A > B"=@n5, "A == B"=@n5, "A < B"=@n3)
  n6 = ScalarParameter() {defaultValue=5, parameterName=Dirt-GrassSpawnLa}
  n7 = ScalarParameter() {defaultValue=-1, parameterName=Dirt-GrassSpawnLb}
  n8 = NamedRerouteUsage()
  n9 = Max(A=Saturate(None=LinearInterpolate(A=@n6, B=@n7, Alpha=@n8)), B=LandscapeLayerSample() {parameterName=Mud})
  n10 = LandscapeLayerSample() {parameterName=Sand}
  n11 = LandscapeLayerSample() {parameterName=Asphalt}
  n12 = LinearInterpolate(
   A=ScalarParameter() {defaultValue=-0.5, parameterName=BiomeB-FFSpawnLa}
   B=ScalarParameter() {defaultValue=1.5, parameterName=BiomeB-FFSpawnLb}
   Alpha=LandscapeLayerSample() {desc=ForestFloor, parameterName=BiomeB}
  )
  n13 = Saturate(None=MaterialFunctionCall(
    Transition Phase=Saturate(None=@n12)
    Height Texture=NamedRerouteUsage()
    Contrast=ScalarParameter() {defaultValue=0.5, parameterName=BiomeB-FFSpawnCont}
   ) {desc=ForestFloor, materialFunction=&HeightLerp}.Alpha)
  n14 = Max(A=Max(A=Max(A=@n9, B=@n10), B=Ceil(None=@n11)), B=@n13)
  n15 = NamedRerouteUsage()
  MaterialExpressionReroute_15 = Reroute(None=OneMinus(None=Max(A=@n15, B=@n14)))
  n16 = LandscapeLayerSample() {parameterName=BiomeC}
  n17 = Constant() {r=2}
  MaterialExpressionSaturate_5 = Saturate(None=Add(A=Multiply(A=@n16, B=@n17), B=Constant() {r=-0.5}))
  MaterialExpressionConstant_1 = Constant()
  n18 = LandscapeLayerSample() {parameterName=Sand}
  n19 = ComponentMask(None=WorldPosition().XYZ) {b=true}
  n20 = Subtract(A=@n19, B=Constant() {r=-5100})
  n21 = Constant() {r=1}
  n22 = OneMinus(None=Saturate(None=Divide(A=@n20, B=@n21)))
  n23 = LandscapeLayerSample() {parameterName=BiomeD}
  MaterialExpressionMax_7 = Max(A=Subtract(A=Subtract(A=@n18, B=@n15), B=@n22), B=Subtract(A=@n23, B=@n15))
  n24 = Subtract(A=@n19, B=Constant() {r=-4600})
  n25 = OneMinus(None=Saturate(None=Divide(A=@n24, B=@n21)))
  MaterialExpressionSubtract_7 = Subtract(A=Multiply(A=OneMinus(None=@n23), B=Subtract(A=@n25, B=@n22)), B=@n15)
  LandscapeGrassOutput!(
   Grass_Standard=@MaterialExpressionReroute_15
   Bush=@MaterialExpressionReroute_15
   FarmFlowers=@MaterialExpressionSaturate_5
   Scree=@MaterialExpressionConstant_1
   Scree_Grass=@MaterialExpressionConstant_1
   Heather=@MaterialExpressionConstant_1
   SwampGrass=@MaterialExpressionConstant_1
   SwampReeds=@MaterialExpressionConstant_1
   Forest_Grass=@MaterialExpressionConstant_1
   Beach=@MaterialExpressionMax_7
   Ocean=@MaterialExpressionSubtract_7
   Nightmare-Grass=@MaterialExpressionConstant_1
   Nightmare-Leaves=@MaterialExpressionConstant_1
   Corrupted2=@MaterialExpressionConstant_1
  ) {
   desc="This must have the exact inputs in every biome material!\r\nThis data is baked into the level data when the level is\r\nsaved, so any changes to this require the affected level\r\nto be saved && submitted."
   grassTypes [
    Grass_Standard {
     grassType=/Game/Athena/Apollo/Environments/Landscape/SmallFoliage/DataAssets/LandscapeGrassType/LandscapeGrassType_StandardGrass
     input {
      expression=&self:MaterialExpressionReroute_15
      inputName=None
      mask=0
      maskA=0
      maskB=0
      maskG=0
      maskR=0
      outputIndex=0
     }
    }
    Bush {
     grassType=/Game/Athena/Environments/Landscape/Foliage/Athena_Foliage_Bushes
     input {
      expression=&self:MaterialExpressionReroute_15
      inputName=None
      mask=0
      maskA=0
      maskB=0
      maskG=0
      maskR=0
      outputIndex=0
     }
    }
    FarmFlowers {
     grassType=/Game/Athena/Apollo/Environments/Landscape/SmallFoliage/DataAssets/LandscapeGrassType/LandscapeGrassType_FarmFlowers
     input {
      expression=&self:MaterialExpressionSaturate_5
      inputName=None
      mask=0
      maskA=0
      maskB=0
      maskG=0
      maskR=0
      outputIndex=0
     }
    }
    Scree {
     grassType=/Game/Athena/Apollo/Environments/Landscape/SmallFoliage/DataAssets/LandscapeGrassType/LandscapeGrassType_Scree
     input {
      expression=&self:MaterialExpressionConstant_1
      inputName=None
      mask=0
      maskA=0
      maskB=0
      maskG=0
      maskR=0
      outputIndex=0
     }
    }
    Scree_Grass {
     grassType=/Game/Athena/Apollo/Environments/Landscape/SmallFoliage/DataAssets/LandscapeGrassType/LandscapeGrassType_ScreeGrass
     input {
      expression=&self:MaterialExpressionConstant_1
      inputName=None
      mask=0
      maskA=0
      maskB=0
      maskG=0
      maskR=0
      outputIndex=0
     }
    }
    Heather {
     grassType=/Game/Athena/Apollo/Environments/Landscape/SmallFoliage/DataAssets/LandscapeGrassType/LandscapeGrassType_Heather
     input {
      expression=&self:MaterialExpressionConstant_1
      inputName=None
      mask=0
      maskA=0
      maskB=0
      maskG=0
      maskR=0
      outputIndex=0
     }
    }
    SwampGrass {
     grassType=/Game/Athena/Apollo/Environments/Landscape/SmallFoliage/DataAssets/LandscapeGrassType/LandscapeGrassType_SwampGrass
     input {
      expression=&self:MaterialExpressionConstant_1
      inputName=None
      mask=0
      maskA=0
      maskB=0
      maskG=0
      maskR=0
      outputIndex=0
     }
    }
    SwampReeds {
     grassType=/Game/Athena/Apollo/Environments/Landscape/SmallFoliage/DataAssets/LandscapeGrassType/LandscapeGrassType_SwampReeds
     input {
      expression=&self:MaterialExpressionConstant_1
      inputName=None
      mask=0
      maskA=0
      maskB=0
      maskG=0
      maskR=0
      outputIndex=0
     }
    }
    Forest_Grass {
     grassType=/Game/Athena/Apollo/Environments/Landscape/SmallFoliage/DataAssets/LandscapeGrassType/LandscapeGrassType_ForestGrass
     input {
      expression=&self:MaterialExpressionConstant_1
      inputName=None
      mask=0
      maskA=0
      maskB=0
      maskG=0
      maskR=0
      outputIndex=0
     }
    }
    Beach {
     grassType=/Game/Athena/Apollo/Environments/Landscape/SmallFoliage/DataAssets/LandscapeGrassType/LandscapeGrassType_Beach
     input {
      expression=&self:MaterialExpressionMax_7
      inputName=None
      mask=0
      maskA=0
      maskB=0
      maskG=0
      maskR=0
      outputIndex=0
     }
    }
    Ocean {
     grassType=/Game/Athena/Apollo/Environments/Landscape/SmallFoliage/DataAssets/LandscapeGrassType/LandscapeGrassType_Ocean
     input {
      expression=&self:MaterialExpressionSubtract_7
      inputName=None
      mask=0
      maskA=0
      maskB=0
      maskG=0
      maskR=0
      outputIndex=0
     }
    }
    Nightmare-Grass {
     grassType=/Game/Athena/Apollo/Environments/Landscape/SmallFoliage/DataAssets/LandscapeGrassType/LandscapeGrassType_Nightmare
     input {
      expression=&self:MaterialExpressionConstant_1
      inputName=None
      mask=0
      maskA=0
      maskB=0
      maskG=0
      maskR=0
      outputIndex=0
     }
    }
    Nightmare-Leaves {
     grassType=/Game/Athena/Apollo/Environments/Landscape/SmallFoliage/DataAssets/LandscapeGrassType/LandscapeGrassType_Nightmare-Leaves
     input {
      expression=&self:MaterialExpressionConstant_1
      inputName=None
      mask=0
      maskA=0
      maskB=0
      maskG=0
      maskR=0
      outputIndex=0
     }
    }
    Corrupted2 {
     grassType=/Game/Athena/Apollo/Environments/Landscape/SmallFoliage/DataAssets/LandscapeGrassType/LandscapeGrassType_Corrupted2
     input {
      expression=&self:MaterialExpressionConstant_1
      inputName=None
      mask=0
      maskA=0
      maskB=0
      maskG=0
      maskR=0
      outputIndex=0
     }
    }
   ]
  }
  n26 = NamedRerouteUsage()
  n27 = OneMinus(None=@n26) {desc="White=Cliffs"}
  n28 = LandscapeLayerSample() {parameterName=Gravel}
  n29 = LandscapeLayerSample() {parameterName=Mud}
  n30 = Saturate(None=Multiply(A=Max(A=@n28, B=@n29), B=Constant() {r=2}))
  n31 = Subtract(A=@n30, B=@n27) {desc="Remove Cliff from Gravel,Mud,Sand"}
  n32 = LandscapeLayerSample() {parameterName=Sand}
  n33 = Saturate(None=Multiply(A=LandscapeLayerSample() {parameterName=Asphalt}, B=Constant() {r=3}))
  n34 = Max(A=Max(A=@MaterialExpressionSaturate_5, B=@n13), B=@n33)
  MaterialExpressionSaturate_12 = Saturate(None=Subtract(A=Saturate(None=@n31), B=Max(A=@n32, B=@n34)) {desc="Remove ForestFloor, Flowers, Asphalt, Sand"}) {desc=Final Dirt mask}
  MaterialExpressionReroute_25 = Reroute(None=@n27) {desc=Final Stone mask}
  n35 = Subtract(A=@n26, B=Max(A=@n32, B=@n30)) {desc="Remove Gravel,Mud,Sand"}
  MaterialExpressionSaturate_14 = Saturate(None=Subtract(A=Saturate(None=@n35), B=@n33) {desc=Remove Asphalt}) {desc=Final Grass mask}
  MaterialExpressionReroute_29 = Reroute(None=@n33)
  n36 = Saturate(None=Subtract(A=@n32, B=@n27) {desc=Remove Cliff})
  MaterialExpressionReroute_42 = Reroute(None=Saturate(None=Subtract(A=@n36, B=@n34) {desc="Remove ForestFloor, Flowers, Asphalt"}))
  LandscapePhysicalMaterialOutput!(
   Dirt=@MaterialExpressionSaturate_12
   Stone=@MaterialExpressionReroute_25
   Grass=@MaterialExpressionSaturate_14
   Asphalt=@MaterialExpressionReroute_29
   Sand=@MaterialExpressionReroute_42
  ) {inputs [
    {
     physicalMaterial=/Game/FortressPhysicalMaterials/Dirt
     input {
      expression=&self:MaterialExpressionSaturate_12
      inputName=None
      mask=0
      maskA=0
      maskB=0
      maskG=0
      maskR=0
      outputIndex=0
     }
    }
    {
     physicalMaterial=/Game/FortressPhysicalMaterials/Stone
     input {
      expression=&self:MaterialExpressionReroute_25
      inputName=None
      mask=0
      maskA=0
      maskB=0
      maskG=0
      maskR=0
      outputIndex=0
     }
    }
    {
     physicalMaterial=/Game/FortressPhysicalMaterials/Grass
     input {
      expression=&self:MaterialExpressionSaturate_14
      inputName=None
      mask=0
      maskA=0
      maskB=0
      maskG=0
      maskR=0
      outputIndex=0
     }
    }
    {
     physicalMaterial=/Game/FortressPhysicalMaterials/Asphalt
     input {
      expression=&self:MaterialExpressionReroute_29
      inputName=None
      mask=0
      maskA=0
      maskB=0
      maskG=0
      maskR=0
      outputIndex=0
     }
    }
    {
     physicalMaterial=/Game/FortressPhysicalMaterials/Sand
     input {
      expression=&self:MaterialExpressionReroute_42
      inputName=None
      mask=0
      maskA=0
      maskB=0
      maskG=0
      maskR=0
      outputIndex=0
     }
    }
   ]}
  n37 = LinearInterpolate(A=@MaterialExpressionSaturate_12, B=@MaterialExpressionReroute_25, Alpha=ScalarParameter() {parameterName=DEBUG-PhysMat1})
  n38 = ScalarParameter() {parameterName=DEBUG-PhysMat2}
  LinearInterpolate(
   A=LinearInterpolate(
    A=LinearInterpolate(A=@n37, B=@MaterialExpressionSaturate_14, Alpha=@n38)
    B=@MaterialExpressionReroute_29
    Alpha=ScalarParameter() {parameterName=DEBUG-PhysMat3}
   )
   B=@MaterialExpressionReroute_42
   Alpha=ScalarParameter() {defaultValue=1, parameterName=DEBUG-PhysMat4}
  )
  n39 = GetMaterialAttributes(None=MaterialFunctionCall!(PutSandBeforeGravelOverride, UseSnow=StaticBool(), UseOceanLevel) {materialFunction=/Game/Athena/Apollo/Environments/Landscape/Materials/Biome/CommonMaterials/Mobile/MF_Mobile_CommonLandscapeLayers}.Material) {attributeGetTypes [#1]}
  MaterialExpressionConstant_10 = Constant()
  n40 = LandscapeLayerBlend!(
   Layer Gravel=@MaterialExpressionConstant_10
   Layer Mud=@MaterialExpressionConstant_10
   Layer Sand=@MaterialExpressionConstant_10
   Layer Asphalt=@MaterialExpressionConstant_10
  ) {layers [
    {
     blendType=LB_WeightBlend
     constHeightInput=0
     layerName=Gravel
     previewWeight=0
     constLayerInput=<0 0 0>
     heightInput {expression=null, inputName=None, mask=0, maskA=0, maskB=0, maskG=0, maskR=0, outputIndex=0}
     layerInput {
      expression=&self:MaterialExpressionConstant_10
      inputName=None
      mask=0
      maskA=0
      maskB=0
      maskG=0
      maskR=0
      outputIndex=0
     }
    }
    {
     blendType=LB_WeightBlend
     constHeightInput=0
     layerName=Mud
     previewWeight=0
     constLayerInput=<0 0 0>
     heightInput {expression=null, inputName=None, mask=0, maskA=0, maskB=0, maskG=0, maskR=0, outputIndex=0}
     layerInput {
      expression=&self:MaterialExpressionConstant_10
      inputName=None
      mask=0
      maskA=0
      maskB=0
      maskG=0
      maskR=0
      outputIndex=0
     }
    }
    {
     blendType=LB_WeightBlend
     constHeightInput=0
     layerName=Sand
     previewWeight=0
     constLayerInput=<0 0 0>
     heightInput {expression=null, inputName=None, mask=0, maskA=0, maskB=0, maskG=0, maskR=0, outputIndex=0}
     layerInput {
      expression=&self:MaterialExpressionConstant_10
      inputName=None
      mask=0
      maskA=0
      maskB=0
      maskG=0
      maskR=0
      outputIndex=0
     }
    }
    {
     blendType=LB_WeightBlend
     constHeightInput=0
     layerName=Asphalt
     previewWeight=0
     constLayerInput=<0 0 0>
     heightInput {expression=null, inputName=None, mask=0, maskA=0, maskB=0, maskG=0, maskR=0, outputIndex=0}
     layerInput {
      expression=&self:MaterialExpressionConstant_10
      inputName=None
      mask=0
      maskA=0
      maskB=0
      maskG=0
      maskR=0
      outputIndex=0
     }
    }
   ]}
  n41 = SetMaterialAttributes!(MaterialAttributes=@n39.MaterialAttributes, Base Color=Add(A=@n39.Base Color, B=Multiply(A=@n40) {constB=0})) {desc="This must remain in material,\r\nnot a function.", attributeSetTypes [#1]}
  n42 = RuntimeVirtualTextureSampleParameter() {
   group=RVT
   materialType=BaseColor_Normal_Specular_YCoCg
   parameterName=RVT Sample Target
   virtualTexture=/Game/Athena/Apollo/Maps/Landscape/VT/RVT_Apollo_Landscape
  }
  n43 = SetMaterialAttributes!(MaterialAttributes, Base Color=@n42.BaseColor, Specular=@n42.BaseColor, Roughness=@n42.BaseColor, Normal=@n42.BaseColor) {attributeSetTypes [
    #1
    #2
    #3
    #4
   ]}
  n44 = NamedRerouteUsage()
  n45 = Constant() {r=6}
  n46 = Saturate(None=Multiply(A=Power(Base=@n44, Exp=@n45), B=Constant() {r=20}))
  n47 = Saturate(None=LinearInterpolate(A=Constant() {r=0.03}, B=Constant() {r=4}, Alpha=NamedRerouteUsage()))
  n48 = Add(A=Constant() {r=0.4}, B=Saturate(None=Add(A=@n46, B=@n47)))
  n49 = MaterialFunctionCall!(AdditionalGrassBorderColorMask=LandscapeLayerSample() {desc=ForestFloor, parameterName=BiomeB}, PutSandBeforeGravel, UseOceanLevel) {materialFunction=/Game/Athena/Apollo/Environments/Landscape/Materials/Biome/CommonMaterials/MF_CommonLandscapeLayers}
  n50 = MaterialFunctionCall() {materialFunction=/Game/Athena/Apollo/Environments/Landscape/Materials/Biome/UniqueBiomeMaterialFunctions/MF_Biome_Forest}
  n51 = MaterialFunctionCall!(
   Material 1=@n49.Material
   Height Texture 1=@n49.Material
   Material 2=@n50.Material
   Height Texture 2=NamedRerouteDeclaration(None=@n50.Height) {name=MF_Biome_Forest Height, nodeColor=<rgba 0.5 0 1 1>}
   Transistion Phase=LandscapeLayerSample() {desc=ForestFloor, parameterName=BiomeB}
   UseContrast
   Contrast
   AddLayerMask
  ) {materialFunction=/Game/Athena/Apollo/Environments/Landscape/Materials/Biome/CommonMaterials/MF_BlendLandscapeLayers}
  n52 = MaterialFunctionCall!(
   GrassHeightmap=NamedRerouteUsage()
   GrassNoiseMask=NamedRerouteUsage()
   GrassBaseColorDetail=NamedRerouteUsage()
   GrassMaterial=NamedRerouteUsage()
  ) {materialFunction=/Game/Athena/Apollo/Environments/Landscape/Materials/Biome/UniqueBiomeMaterialFunctions/MF_Biome_FarmGrass}
  n53 = MaterialFunctionCall(
   Material 1=@n51.BlendedMaterial
   Height Texture 1=@n51.BlendedMaterial
   Material 2=@n52.Material
   Height Texture 2=@n52.Material
   Transistion Phase=LandscapeLayerSample() {parameterName=BiomeC}
  ) {materialFunction=&MF_BlendLandscapeLayers}
  n54 = GetMaterialAttributes(None=@n53.BlendedMaterial) {attributeGetTypes [#1]}
  n55 = LinearInterpolate(A=Multiply(A=@n54.Base Color, B=Saturate(None=@n48)), B=@n54.Base Color, Alpha=LandscapeLayerSample() {parameterName=Sand})
  n56 = LinearInterpolate(A=@n46, B=Constant() {r=0.5}, Alpha=@n47)
  n57 = MaterialFunctionCall!(
   Material=BlendMaterialAttributes(
    A=@n54.MaterialAttributes
    B=SetMaterialAttributes!(MaterialAttributes=@n54.MaterialAttributes, Base Color=@n55, Roughness=@n56) {attributeSetTypes [#1, #3]}
    Alpha=LandscapeLayerSample() {parameterName=BiomeD}
   ) {desc="Vertex blend type altered!", vertexAttributeBlendType=UseA}
   Heightmap=@n53.Resulting Height Map
   NewSlopeMaskMask
   Displacement
   AsphaltLayer=LandscapeLayerSample() {parameterName=Sand}
   UseAlternateCliffMat
   AltCliffHeight
   AltCliffMat
  ) {materialFunction=/Game/Athena/Apollo/Environments/Landscape/Materials/Biome/CommonMaterials/MF_SlopeBlendCliff}
  n58 = MaterialFunctionCall!(
   MobileMaterial=@n41
   HighEndMaterial=MaterialFunctionCall!(Material=@n57.Material) {materialFunction=/Game/Athena/Apollo/Maps/Landscape/VT/MaterialFunctions/TransformTangentToWorld}.Result
  ) {materialFunction=/Game/Athena/Asteria/Environments/Landscape/Materials/MaterialFunctions/FN_RvtWriteSwitch}
  n59 = MaterialFunctionCall(
   NoRVT=@n41
   RVT=StaticSwitchParameter(
    True=MaterialFunctionCall!(RVT=@n43, ByPass RVT=@n58.Result) {materialFunction=/Game/Athena/Apollo/Maps/Landscape/VT/MaterialFunctions/EnableRVT}.Result
    False=@n58.Result
   ) {defaultValue=true, group=RVT Base, parameterName=RVT Enable}
  ) {materialFunction=&FN_RVT_Switch_Material}
  n60 = SetMaterialAttributes!(MaterialAttributes=@n59.Result, Opacity Mask=LandscapeVisibilityMask()) {attributeSetTypes [#5]}
  MaterialFunctionCall!(
   In=ShadingPathSwitch(Default=@n60, Deferred=@n60, Forward=@n60, Mobile=@n59.Result)
   UseHeightFalloff
   OnFireEmissiveOverride
   HeightFalloffRadius
   HeightFalloffHardness
   VariationMasksMobile
   In_Grass
   Is_Stylised_WE
  ) {desc=This should be very last, materialFunction=/Game/Packages/Fortress_SharedMaterials/Base_Material_Functions/MF_Curie_Landscape}
  NamedRerouteDeclaration(None=@n49.GrassNoiseMask) {name=GrassNoiseMask, nodeColor=<rgba 0.5 0 1 1>}
  NamedRerouteDeclaration(None=@n49.GrassBaseColorDetail) {name=GrassBaseColorDetail, nodeColor=<rgba 0.5 0 1 1>}
  NamedRerouteDeclaration(None=@n49.Material) {name=CommonMaterial, nodeColor=<rgba 0.5 0 1 1>}
  NamedRerouteDeclaration(None=@n49.GrassHeightmap) {name=GrassHeightmap, nodeColor=<rgba 0.5 0 1 1>}
  NamedRerouteDeclaration(None=@n49.GravelMaskForGrassFoliage) {name=MaskForGrassFoliage, nodeColor=<rgba 0.5 0 1 1>}
  NamedRerouteDeclaration(None=@n49.GravelHeight) {name=GravelHeight, nodeColor=<rgba 0.5 0 1 1>}
  NamedRerouteDeclaration(None=@n57.HeightmapForGrass) {name=HeightmapForGrass, nodeColor=<rgba 0 1 0.7422 1>}
  NamedRerouteDeclaration(None=@n57.CliffMask) {name=CliffMask, nodeColor=<rgba 0.5 0 1 1>}
  NamedRerouteDeclaration(None=@n57.LandscapeGrassMeshMask) {name=GrassMeshMask, nodeColor=<rgba 0.5 0 1 1>}
  n61 = GetMaterialAttributes(None=@n58.Result) {attributeGetTypes [
    #1
    #2
    #3
    #4
   ]}
  RuntimeVirtualTextureOutput!(
   BaseColor=MaterialFunctionCall!(BaseColor=@n61.Base Color) {materialFunction=/Game/Athena/Apollo/Environments/Landscape/Materials/Biome/CommonMaterials/MF_ReduceRTV-CompressionArtifacts}.BaseColor
   Specular=@n61.Specular
   Roughness=@n61.Specular
   Normal=@n61.Specular
   WorldHeight=ComponentMask(None=WorldPosition().XYZ) {b=true}
   Opacity
   Mask
   Displacement
   Mask4
  )
 ]
}
