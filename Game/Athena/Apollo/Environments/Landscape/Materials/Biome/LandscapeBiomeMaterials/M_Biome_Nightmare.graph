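# Landscape material for the Nightmare biome, blending mud, jungle grass and leaves over common layers via BiomeA/B/C weights, then adding slope-based cliffs.
# The result, or a simpler mobile layer set, is written to a runtime virtual texture, with switchable sampling back, before a final Curie landscape function.
# Layer and cliff masks drive physical materials; layer weights also drive grass spawning, whose input list must match every biome material and bakes on level save.
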Material /Game/Athena/Apollo/Environments/Landscape/Materials/Biome/LandscapeBiomeMaterials/M_Biome_Nightmare
graph {
 properties {bTangentSpaceNormal=false, bUseMaterialAttributes=true, pixelDepthOffsetMode=PDOM_Legacy, usedShadingModels=Default Lit}
 nodes [
  n1 = LandscapeVisibilityMask()
  n2 = Max(A=LandscapeLayerSample() {parameterName=Gravel}, B=LandscapeLayerSample() {parameterName=Mud})
  n3 = Constant() {r=2}
  n4 = Max(A=Saturate(None=Multiply(A=@n2, B=@n3)), B=LandscapeLayerSample() {parameterName=Sand})
  n5 = Ceil(None=LandscapeLayerSample() {parameterName=Asphalt})
  n6 = Max(A=LandscapeLayerSample() {parameterName=BiomeA}, B=LandscapeLayerSample() {parameterName=BiomeB})
  n7 = LandscapeLayerSample() {parameterName=BiomeC}
  n8 = NamedRerouteUsage()
  n9 = Max(A=@n8, B=Max(A=Max(A=@n4, B=@n5), B=Max(A=@n6, B=@n7)))
  MaterialExpressionReroute_4 = Reroute(None=Multiply(A=@n1, B=OneMinus(None=@n9)))
  MaterialExpressionConstant_8 = Constant()
  n10 = LandscapeLayerSample() {parameterName=BiomeB}
  n11 = LandscapeLayerSample() {parameterName=BiomeC}
  MaterialExpressionMultiply_3 = Multiply(A=@n1, B=Subtract(A=Subtract(A=@n10, B=@n11) {constA=2}, B=@n8) {constA=2})
  n12 = LandscapeLayerSample() {parameterName=BiomeC}
  n13 = LandscapeLayerSample() {parameterName=BiomeB}
  MaterialExpressionMultiply_5 = Multiply(A=@n1, B=Subtract(A=Subtract(A=@n12, B=@n13) {constA=2}, B=@n8) {constA=2})
  LandscapeGrassOutput!(
   Grass_Standard=@MaterialExpressionReroute_4
   Bush=@MaterialExpressionReroute_4
   FarmFlowers=@MaterialExpressionConstant_8
   Scree=@MaterialExpressionConstant_8
   Scree_Grass=@MaterialExpressionConstant_8
   Heather=@MaterialExpressionConstant_8
   SwampGrass=@MaterialExpressionConstant_8
   SwampReeds=@MaterialExpressionConstant_8
   Forest_Grass=@MaterialExpressionConstant_8
   Beach=@MaterialExpressionConstant_8
   Ocean=@MaterialExpressionConstant_8
   Nightmare-Grass=@MaterialExpressionMultiply_3
   Nightmare-Leaves=@MaterialExpressionMultiply_5
   Corrupted2=@MaterialExpressionConstant_8
  ) {
   desc="This must have the exact inputs in every biome material!\r\nThis data is baked into the level data when the level is\r\nsaved, so any changes to this require the affected level\r\nto be saved && submitted."
   grassTypes [
    Grass_Standard {
     grassType=/Game/Athena/Apollo/Environments/Landscape/SmallFoliage/DataAssets/LandscapeGrassType/LandscapeGrassType_StandardGrass
     input {
      expression=&self:MaterialExpressionReroute_4
      inputName=None
      mask=0
      maskA=0
      maskB=0
      maskG=0
      maskR=0
      outputIndex=0
     }
    }
    Bush {
     grassType=/Game/Athena/Environments/Landscape/Foliage/Athena_Foliage_Bushes
     input {
      expression=&self:MaterialExpressionReroute_4
      inputName=None
      mask=0
      maskA=0
      maskB=0
      maskG=0
      maskR=0
      outputIndex=0
     }
    }
    FarmFlowers {
     grassType=/Game/Athena/Apollo/Environments/Landscape/SmallFoliage/DataAssets/LandscapeGrassType/LandscapeGrassType_FarmFlowers
     input {
      expression=&self:MaterialExpressionConstant_8
      inputName=None
      mask=0
      maskA=0
      maskB=0
      maskG=0
      maskR=0
      outputIndex=0
     }
    }
    Scree {
     grassType=/Game/Athena/Apollo/Environments/Landscape/SmallFoliage/DataAssets/LandscapeGrassType/LandscapeGrassType_Scree
     input {
      expression=&self:MaterialExpressionConstant_8
      inputName=None
      mask=0
      maskA=0
      maskB=0
      maskG=0
      maskR=0
      outputIndex=0
     }
    }
    Scree_Grass {
     grassType=/Game/Athena/Apollo/Environments/Landscape/SmallFoliage/DataAssets/LandscapeGrassType/LandscapeGrassType_ScreeGrass
     input {
      expression=&self:MaterialExpressionConstant_8
      inputName=None
      mask=0
      maskA=0
      maskB=0
      maskG=0
      maskR=0
      outputIndex=0
     }
    }
    Heather {
     grassType=/Game/Athena/Apollo/Environments/Landscape/SmallFoliage/DataAssets/LandscapeGrassType/LandscapeGrassType_Heather
     input {
      expression=&self:MaterialExpressionConstant_8
      inputName=None
      mask=0
      maskA=0
      maskB=0
      maskG=0
      maskR=0
      outputIndex=0
     }
    }
    SwampGrass {
     grassType=/Game/Athena/Apollo/Environments/Landscape/SmallFoliage/DataAssets/LandscapeGrassType/LandscapeGrassType_SwampGrass
     input {
      expression=&self:MaterialExpressionConstant_8
      inputName=None
      mask=0
      maskA=0
      maskB=0
      maskG=0
      maskR=0
      outputIndex=0
     }
    }
    SwampReeds {
     grassType=/Game/Athena/Apollo/Environments/Landscape/SmallFoliage/DataAssets/LandscapeGrassType/LandscapeGrassType_SwampReeds
     input {
      expression=&self:MaterialExpressionConstant_8
      inputName=None
      mask=0
      maskA=0
      maskB=0
      maskG=0
      maskR=0
      outputIndex=0
     }
    }
    Forest_Grass {
     grassType=/Game/Athena/Apollo/Environments/Landscape/SmallFoliage/DataAssets/LandscapeGrassType/LandscapeGrassType_ForestGrass
     input {
      expression=&self:MaterialExpressionConstant_8
      inputName=None
      mask=0
      maskA=0
      maskB=0
      maskG=0
      maskR=0
      outputIndex=0
     }
    }
    Beach {
     grassType=/Game/Athena/Apollo/Environments/Landscape/SmallFoliage/DataAssets/LandscapeGrassType/LandscapeGrassType_Beach
     input {
      expression=&self:MaterialExpressionConstant_8
      inputName=None
      mask=0
      maskA=0
      maskB=0
      maskG=0
      maskR=0
      outputIndex=0
     }
    }
    Ocean {
     grassType=/Game/Athena/Apollo/Environments/Landscape/SmallFoliage/DataAssets/LandscapeGrassType/LandscapeGrassType_Ocean
     input {
      expression=&self:MaterialExpressionConstant_8
      inputName=None
      mask=0
      maskA=0
      maskB=0
      maskG=0
      maskR=0
      outputIndex=0
     }
    }
    Nightmare-Grass {
     grassType=/Game/Athena/Apollo/Environments/Landscape/SmallFoliage/DataAssets/LandscapeGrassType/LandscapeGrassType_Nightmare
     input {
      expression=&self:MaterialExpressionMultiply_3
      inputName=None
      mask=0
      maskA=0
      maskB=0
      maskG=0
      maskR=0
      outputIndex=0
     }
    }
    Nightmare-Leaves {
     grassType=/Game/Athena/Apollo/Environments/Landscape/SmallFoliage/DataAssets/LandscapeGrassType/LandscapeGrassType_Nightmare-Leaves
     input {
      expression=&self:MaterialExpressionMultiply_5
      inputName=None
      mask=0
      maskA=0
      maskB=0
      maskG=0
      maskR=0
      outputIndex=0
     }
    }
    Corrupted2 {
     grassType=/Game/Athena/Apollo/Environments/Landscape/SmallFoliage/DataAssets/LandscapeGrassType/LandscapeGrassType_Corrupted2
     input {
      expression=&self:MaterialExpressionConstant_8
      inputName=None
      mask=0
      maskA=0
      maskB=0
      maskG=0
      maskR=0
      outputIndex=0
     }
    }
   ]
  }
  n14 = NamedRerouteUsage()
  n15 = OneMinus(None=@n14) {desc="White=Cliffs"}
  n16 = Max(
   A=Max(A=LandscapeLayerSample() {parameterName=BiomeA}, B=LandscapeLayerSample() {parameterName=BiomeB})
   B=LandscapeLayerSample() {parameterName=BiomeC}
  )
  n17 = Max(A=LandscapeLayerSample() {parameterName=Gravel}, B=LandscapeLayerSample() {parameterName=Mud})
  n18 = Constant() {r=2}
  n19 = Max(A=@n16, B=Saturate(None=Multiply(A=@n17, B=@n18)))
  n20 = Subtract(A=@n19, B=@n15) {desc="Remove Cliff from Gravel,Mud, Sand, Snow"}
  n21 = LandscapeLayerSample() {parameterName=Sand}
  n22 = Saturate(None=Multiply(A=LandscapeLayerSample() {parameterName=Asphalt}, B=Constant() {r=3}))
  MaterialExpressionSaturate_12 = Saturate(None=Subtract(A=Saturate(None=@n20), B=Max(A=@n21, B=@n22)) {desc="Remove Asphalt, Sand"}) {desc=Final Dirt mask}
  MaterialExpressionReroute_51 = Reroute(None=@n15) {desc=Final Stone mask}
  n23 = Subtract(A=@n14, B=Max(A=@n21, B=@n19)) {desc="Remove Gravel,Mud, Sand, Snow"}
  MaterialExpressionSaturate_14 = Saturate(None=Subtract(A=Saturate(None=@n23), B=@n22) {desc=Remove Asphalt}) {desc=Final Grass mask}
  MaterialExpressionReroute_61 = Reroute(None=@n22)
  n24 = Saturate(None=Subtract(A=@n21, B=@n15) {desc=Remove Cliff})
  n25 = Max(A=@n16, B=@n22)
  MaterialExpressionReroute_42 = Reroute(None=Saturate(None=Subtract(A=@n24, B=@n25)))
  LandscapePhysicalMaterialOutput!(
   Dirt=@MaterialExpressionSaturate_12
   Stone=@MaterialExpressionReroute_51
   Grass=@MaterialExpressionSaturate_14
   Asphalt=@MaterialExpressionReroute_61
   Sand=@MaterialExpressionReroute_42
  ) {inputs [
    {
     physicalMaterial=/Game/FortressPhysicalMaterials/Dirt
     input {
      expression=&self:MaterialExpressionSaturate_12
      inputName=None
      mask=0
      maskA=0
      maskB=0
      maskG=0
      maskR=0
      outputIndex=0
     }
    }
    {
     physicalMaterial=/Game/FortressPhysicalMaterials/Stone
     input {
      expression=&self:MaterialExpressionReroute_51
      inputName=None
      mask=0
      maskA=0
      maskB=0
      maskG=0
      maskR=0
      outputIndex=0
     }
    }
    {
     physicalMaterial=/Game/FortressPhysicalMaterials/Grass
     input {
      expression=&self:MaterialExpressionSaturate_14
      inputName=None
      mask=0
      maskA=0
      maskB=0
      maskG=0
      maskR=0
      outputIndex=0
     }
    }
    {
     physicalMaterial=/Game/FortressPhysicalMaterials/Asphalt
     input {
      expression=&self:MaterialExpressionReroute_61
      inputName=None
      mask=0
      maskA=0
      maskB=0
      maskG=0
      maskR=0
      outputIndex=0
     }
    }
    {
     physicalMaterial=/Game/FortressPhysicalMaterials/Sand
     input {
      expression=&self:MaterialExpressionReroute_42
      inputName=None
      mask=0
      maskA=0
      maskB=0
      maskG=0
      maskR=0
      outputIndex=0
     }
    }
   ]}
  n26 = LinearInterpolate(A=@MaterialExpressionSaturate_12, B=@MaterialExpressionReroute_51, Alpha=ScalarParameter() {parameterName=DEBUG-PhysMat1})
  n27 = ScalarParameter() {parameterName=DEBUG-PhysMat2}
  LinearInterpolate(
   A=LinearInterpolate(
    A=LinearInterpolate(A=@n26, B=@MaterialExpressionSaturate_14, Alpha=@n27)
    B=@MaterialExpressionReroute_61
    Alpha=ScalarParameter() {parameterName=DEBUG-PhysMat3}
   )
   B=@MaterialExpressionReroute_42
   Alpha=ScalarParameter() {defaultValue=1, parameterName=DEBUG-PhysMat4}
  )
  n28 = GetMaterialAttributes(None=MaterialFunctionCall!(PutSandBeforeGravelOverride, UseSnow=StaticBool(), UseOceanLevel) {materialFunction=/Game/Athena/Apollo/Environments/Landscape/Materials/Biome/CommonMaterials/Mobile/MF_Mobile_CommonLandscapeLayers}.Material) {attributeGetTypes [#1]}
  MaterialExpressionConstant_13 = Constant()
  n29 = LandscapeLayerBlend!(
   Layer Gravel=@MaterialExpressionConstant_13
   Layer Mud=@MaterialExpressionConstant_13
   Layer Sand=@MaterialExpressionConstant_13
   Layer Asphalt=@MaterialExpressionConstant_13
  ) {layers [
    {
     blendType=LB_WeightBlend
     constHeightInput=0
     layerName=Gravel
     previewWeight=0
     constLayerInput=<0 0 0>
     heightInput {expression=null, inputName=None, mask=0, maskA=0, maskB=0, maskG=0, maskR=0, outputIndex=0}
     layerInput {
      expression=&self:MaterialExpressionConstant_13
      inputName=None
      mask=0
      maskA=0
      maskB=0
      maskG=0
      maskR=0
      outputIndex=0
     }
    }
    {
     blendType=LB_WeightBlend
     constHeightInput=0
     layerName=Mud
     previewWeight=0
     constLayerInput=<0 0 0>
     heightInput {expression=null, inputName=None, mask=0, maskA=0, maskB=0, maskG=0, maskR=0, outputIndex=0}
     layerInput {
      expression=&self:MaterialExpressionConstant_13
      inputName=None
      mask=0
      maskA=0
      maskB=0
      maskG=0
      maskR=0
      outputIndex=0
     }
    }
    {
     blendType=LB_WeightBlend
     constHeightInput=0
     layerName=Sand
     previewWeight=0
     constLayerInput=<0 0 0>
     heightInput {expression=null, inputName=None, mask=0, maskA=0, maskB=0, maskG=0, maskR=0, outputIndex=0}
     layerInput {
      expression=&self:MaterialExpressionConstant_13
      inputName=None
      mask=0
      maskA=0
      maskB=0
      maskG=0
      maskR=0
      outputIndex=0
     }
    }
    {
     blendType=LB_WeightBlend
     constHeightInput=0
     layerName=Asphalt
     previewWeight=0
     constLayerInput=<0 0 0>
     heightInput {expression=null, inputName=None, mask=0, maskA=0, maskB=0, maskG=0, maskR=0, outputIndex=0}
     layerInput {
      expression=&self:MaterialExpressionConstant_13
      inputName=None
      mask=0
      maskA=0
      maskB=0
      maskG=0
      maskR=0
      outputIndex=0
     }
    }
   ]}
  n30 = SetMaterialAttributes!(MaterialAttributes=@n28.MaterialAttributes, Base Color=Add(A=@n28.Base Color, B=Multiply(A=@n29) {constB=0})) {attributeSetTypes [#1]}
  n31 = RuntimeVirtualTextureSampleParameter() {
   group=RVT
   materialType=BaseColor_Normal_Specular_YCoCg
   parameterName=RVT Sample Target
   virtualTexture=/Game/Athena/Apollo/Maps/Landscape/VT/RVT_Apollo_Landscape
  }
  n32 = SetMaterialAttributes!(MaterialAttributes, Base Color=@n31.BaseColor, Specular=@n31.BaseColor, Roughness=@n31.BaseColor, Normal=@n31.BaseColor) {attributeSetTypes [
    #1
    #2
    #3
    #4
   ]}
  n33 = MaterialFunctionCall!(AdditionalGrassBorderColorMask, PutSandBeforeGravel, UseOceanLevel) {materialFunction=/Game/Athena/Apollo/Environments/Landscape/Materials/Biome/CommonMaterials/MF_CommonLandscapeLayers}
  n34 = MaterialFunctionCall!(
   Scale=Constant() {r=1.5}
   Diffuse=TextureObject() {texture=/Game/Environments/Apollo/Sets/Nightmare/Textures/T_Nightmare_Mud_D_01}
   Normal=TextureObject() {samplerType=SAMPLERTYPE_Normal, texture=/Game/Environments/Apollo/Sets/Nightmare/Textures/T_Nightmare_Mud_N_01}
   S_Mask=TextureObject() {samplerType=SAMPLERTYPE_Masks, texture=/Game/Environments/Apollo/Sets/Nightmare/Textures/T_Nightmare_Mud_S_01}
  ) {materialFunction=/Game/Athena/Apollo/Environments/Landscape/Materials/Biome/UniqueBiomeMaterialFunctions/MF_Biome_Generic}
  n35 = MaterialFunctionCall!(
   Material 1=@n33.Material
   Height Texture 1=@n33.Material
   Material 2=@n34.Material
   Height Texture 2=@n34.Material
   Transistion Phase=LandscapeLayerSample() {parameterName=BiomeA}
   UseContrast
   Contrast
   AddLayerMask
  ) {materialFunction=/Game/Athena/Apollo/Environments/Landscape/Materials/Biome/CommonMaterials/MF_BlendLandscapeLayers}
  n36 = MaterialFunctionCall(
   Scale=Constant() {r=1.5}
   Diffuse=TextureObject() {texture=/Game/Environments/Apollo/Sets/Nightmare/Textures/T_Nightmare_JungleGrass_D_01}
   Normal=TextureObject() {samplerType=SAMPLERTYPE_Normal, texture=/Game/Environments/Apollo/Sets/Nightmare/Textures/T_Nightmare_JungleGrass_N_01}
   S_Mask=TextureObject() {samplerType=SAMPLERTYPE_Masks, texture=/Game/Environments/Apollo/Sets/Nightmare/Textures/T_Nightmare_JungleGrass_S_01}
  ) {materialFunction=&MF_Biome_Generic}
  n37 = MaterialFunctionCall(
   Material 1=@n35.BlendedMaterial
   Height Texture 1=@n35.BlendedMaterial
   Material 2=@n36.Material
   Height Texture 2=@n36.Material
   Transistion Phase=LandscapeLayerSample() {parameterName=BiomeB}
  ) {materialFunction=&MF_BlendLandscapeLayers}
  n38 = MaterialFunctionCall(
   Scale=Constant() {r=1.5}
   Diffuse=TextureObject() {texture=/Game/Environments/Apollo/Sets/Nightmare/Textures/T_Nightmare_Leaves_D_01}
   Normal=TextureObject() {samplerType=SAMPLERTYPE_Normal, texture=/Game/Environments/Apollo/Sets/Nightmare/Textures/T_Nightmare_Leaves_N_01}
   S_Mask=TextureObject() {samplerType=SAMPLERTYPE_Masks, texture=/Game/Environments/Apollo/Sets/Nightmare/Textures/T_Nightmare_Leaves_S_01}
  ) {materialFunction=&MF_Biome_Generic}
  n39 = MaterialFunctionCall(
   Material 1=@n37.BlendedMaterial
   Height Texture 1=@n37.BlendedMaterial
   Material 2=@n38.Material
   Height Texture 2=@n38.Material
   Transistion Phase=LandscapeLayerSample() {parameterName=BiomeC}
  ) {materialFunction=&MF_BlendLandscapeLayers}
  n40 = MaterialFunctionCall!(
   Material=@n39.BlendedMaterial
   Heightmap=@n39.BlendedMaterial
   NewSlopeMaskMask
   Displacement
   AsphaltLayer=LandscapeLayerSample() {parameterName=Sand}
   UseAlternateCliffMat
   AltCliffHeight
   AltCliffMat
  ) {materialFunction=/Game/Athena/Apollo/Environments/Landscape/Materials/Biome/CommonMaterials/MF_SlopeBlendCliff}
  n41 = MaterialFunctionCall!(
   MobileMaterial=@n30
   HighEndMaterial=MaterialFunctionCall!(Material=@n40.Material) {materialFunction=/Game/Athena/Apollo/Maps/Landscape/VT/MaterialFunctions/TransformTangentToWorld}.Result
  ) {materialFunction=/Game/Athena/Asteria/Environments/Landscape/Materials/MaterialFunctions/FN_RvtWriteSwitch}
  n42 = MaterialFunctionCall(
   NoRVT=@n30
   RVT=StaticSwitchParameter(
    True=MaterialFunctionCall!(RVT=@n32, ByPass RVT=@n41.Result) {materialFunction=/Game/Athena/Apollo/Maps/Landscape/VT/MaterialFunctions/EnableRVT}.Result
    False=@n41.Result
   ) {defaultValue=true, group=RVT Base, parameterName=RVT Enable}
  ) {materialFunction=&FN_RVT_Switch_Material}
  n43 = SetMaterialAttributes!(MaterialAttributes=@n42.Result, Opacity Mask=LandscapeVisibilityMask()) {attributeSetTypes [#5]}
  MaterialFunctionCall!(
   In=ShadingPathSwitch(Default=@n43, Deferred=@n43, Forward=@n43, Mobile=@n42.Result)
   UseHeightFalloff
   OnFireEmissiveOverride
   HeightFalloffRadius
   HeightFalloffHardness
   VariationMasksMobile
   In_Grass
   Is_Stylised_WE
  ) {desc=This should be very last, materialFunction=/Game/Packages/Fortress_SharedMaterials/Base_Material_Functions/MF_Curie_Landscape}
  Max(A=@n7, B=LandscapeLayerSample() {parameterName=BiomeD})
  NamedRerouteDeclaration(None=@n40.HeightmapForGrass) {name=HeightmapForGrass, nodeColor=<rgba 1 1 0 1>}
  NamedRerouteDeclaration(None=@n40.CliffMask) {name=CliffMask, nodeColor=<rgba 0.5 0 1 1>}
  NamedRerouteDeclaration(None=@n40.LandscapeGrassMeshMask) {name=GrassMeshMask, nodeColor=<rgba 0.5 0 1 1>}
  n44 = GetMaterialAttributes(None=@n41.Result) {attributeGetTypes [
    #1
    #2
    #3
    #4
   ]}
  RuntimeVirtualTextureOutput!(
   BaseColor=MaterialFunctionCall!(BaseColor=@n44.Base Color) {materialFunction=/Game/Athena/Apollo/Environments/Landscape/Materials/Biome/CommonMaterials/MF_ReduceRTV-CompressionArtifacts}.BaseColor
   Specular=@n44.Specular
   Roughness=@n44.Specular
   Normal=@n44.Specular
   WorldHeight=ComponentMask(None=WorldPosition().XYZ) {b=true}
   Opacity
   Mask
   Displacement
   Mask4
  )
 ]
}
