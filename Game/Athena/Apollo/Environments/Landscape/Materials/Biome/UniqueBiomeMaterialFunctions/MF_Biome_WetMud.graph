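# Landscape biome material function that builds wet mud material attributes for Apollo terrain and passes Height through.
# Diffuse is divided by BiomeColorDiv, then tinted between dark and regular biome colors by Height.
# PondScumLayer color blends in around a hardcoded world Z level, using absolute distance from it.
# Roughness follows Height and saturates below that level, specular is a low constant, and normal XY is amplified.
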MaterialFunction /Game/Athena/Apollo/Environments/Landscape/Materials/Biome/UniqueBiomeMaterialFunctions/MF_Biome_WetMud
graph {nodes [
  Constant() {r=-0.2}
  Constant() {r=0.02}
  Constant() {r=1}
  Constant() {r=-0.4}
  n1 = FunctionInput() {bUsePreviewValueAsDefault=true, inputName=BiomeColorDiv, sortPriority=2, previewValue=<1 1 1 1>}
  n2 = Divide(A=FunctionInput() {bUsePreviewValueAsDefault=true, inputName=Diffuse}, B=@n1)
  n3 = FunctionInput() {bUsePreviewValueAsDefault=true, inputName=BiomeColorDark, sortPriority=4, previewValue=<1 1 1 1>}
  n4 = FunctionInput() {bUsePreviewValueAsDefault=true, inputName=Height, inputType=FunctionInput_Scalar, previewValue=<1 1 1 1>}
  n5 = LinearInterpolate(
   A=Multiply(A=@n2, B=@n3)
   B=Multiply(A=FunctionInput() {bUsePreviewValueAsDefault=true, inputName=BiomeColor, sortPriority=3, previewValue=<1 1 1 1>}, B=@n2)
   Alpha=@n4
  )
  n6 = Add(A=ComponentMask(None=WorldPosition().XYZ) {b=true}, B=Constant() {r=3909})
  n7 = Divide(A=Abs(None=@n6), B=Constant() {r=25}) {constB=10}
  n8 = Add(A=Multiply(A=Saturate(None=@n7), B=Constant() {r=1.5}), B=Constant() {r=-0.25})
  n9 = OneMinus(None=Saturate(None=@n6))
  n10 = Constant() {r=-0.4}
  n11 = Multiply(A=Constant() {r=1}, B=@n4)
  n12 = Constant() {r=-0.2}
  n13 = LinearInterpolate(A=Saturate(None=Add(A=@n10, B=@n11)), B=Saturate(None=Add(A=@n12, B=@n11)))
  n14 = SetMaterialAttributes!(
   MaterialAttributes
   BaseColor=LinearInterpolate(
    A=MaterialFunctionCall!(Attributes, Sim Erasure, Alpha Mask, Depth) {materialFunction=/FortWater/Materials/Functions/PondScumLayer}.Color
    B=Saturate(None=@n5)
    Alpha=Saturate(None=@n8)
   )
   Specular=Constant() {r=0.02}
   Roughness=Saturate(None=Add(A=@n9, B=@n13))
   Normal=Multiply(A=FunctionInput() {bUsePreviewValueAsDefault=true, inputName=Normals, previewValue=<0 0 1 1>}, B=Constant3Vector() {constant=<rgba 3 3 1 1>})
  ) {attributeSetTypes [
    #1
    #2
    #3
    #4
   ]}
  FunctionOutput(None=@n14) {outputName=Material}
  FunctionOutput(None=@n4) {outputName=Height, sortPriority=1}
 ]}
