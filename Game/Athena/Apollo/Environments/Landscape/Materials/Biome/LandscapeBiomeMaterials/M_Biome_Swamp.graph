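# Swamp biome landscape material for the Apollo map, driving terrain shading, grass spawning and physical surface types.
# Common layers height-blend with wet mud, forest floor and swamp grass, then slope-blended cliffs.
# An RVT switch writes or samples that result, with a simpler mobile path, before the final Curie function.
# Layer masks feed baked grass types, with reeds limited to mud in a world-height band, plus prioritized physical materials.
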Material /Game/Athena/Apollo/Environments/Landscape/Materials/Biome/LandscapeBiomeMaterials/M_Biome_Swamp
graph {
 properties {bTangentSpaceNormal=false, bUseMaterialAttributes=true, pixelDepthOffsetMode=PDOM_Legacy, usedShadingModels=Default Lit}
 nodes [
  Constant() {r=3909}
  n1 = LandscapeLayerSample() {parameterName=BiomeC}
  n2 = Saturate(None=Multiply(A=Constant() {r=3}, B=@n1))
  n3 = ScalarParameter() {defaultValue=-0.5, parameterName=BiomeB-FFSpawnLa}
  n4 = ScalarParameter() {defaultValue=1.5, parameterName=BiomeB-FFSpawnLb}
  n5 = LandscapeLayerSample() {desc=ForestFloor, parameterName=BiomeB}
  n6 = MaterialFunctionCall(
   Transition Phase=Saturate(None=LinearInterpolate(A=@n3, B=@n4, Alpha=@n5))
   Height Texture=NamedRerouteUsage()
   Contrast=ScalarParameter() {defaultValue=0.5, parameterName=BiomeB-FFSpawnCont}
  ) {desc=ForestFloor, materialFunction=&HeightLerp}
  n7 = Max(A=LandscapeLayerSample() {parameterName=BiomeA}, B=@n6.Alpha)
  n8 = Saturate(None=LinearInterpolate(
    A=ScalarParameter() {defaultValue=5, parameterName=Dirt-GrassSpawnLa}
    B=ScalarParameter() {defaultValue=-1, parameterName=Dirt-GrassSpawnLb}
    Alpha=NamedRerouteUsage()
   ))
  n9 = LandscapeLayerSample() {parameterName=Mud}
  n10 = Multiply(A=Constant() {r=2}, B=LandscapeLayerSample() {parameterName=Sand})
  n11 = Max(A=Max(A=Max(A=@n8, B=@n9), B=Saturate(None=@n10)), B=Ceil(None=LandscapeLayerSample() {parameterName=Asphalt}))
  n12 = NamedRerouteUsage()
  n13 = OneMinus(None=Max(A=Max(A=@n2, B=@n7), B=Max(A=@n12, B=@n11)))
  n14 = NamedRerouteUsage()
  n15 = Subtract(A=@n13, B=@n14) {desc=Grass_Standard}
  MaterialExpressionReroute_22 = Reroute(None=@n15)
  n16 = Max(A=Min(A=@n14, B=@n13) {desc=Forest grass}, B=@n6.Alpha)
  MaterialExpressionReroute_23 = Reroute(None=Max(A=Max(A=@n15, B=@n16), B=@n1))
  MaterialExpressionConstant_8 = Constant()
  MaterialExpressionReroute_33 = Reroute(None=@n1)
  n17 = ComponentMask(None=WorldPosition().XYZ) {b=true}
  n18 = Add(A=@n17, B=Constant() {r=4000})
  n19 = Constant() {r=500}
  n20 = OneMinus(None=Saturate(None=Divide(A=@n18, B=@n19)))
  n21 = Saturate(None=Add(A=@n17, B=Constant() {r=3990}))
  MaterialExpressionReroute_32 = Reroute(None=Multiply(A=@n9, B=Multiply(A=@n20, B=@n21)))
  MaterialExpressionReroute_21 = Reroute(None=@n16)
  LandscapeGrassOutput!(
   Grass_Standard=@MaterialExpressionReroute_22
   Bush=@MaterialExpressionReroute_23
   FarmFlowers=@MaterialExpressionConstant_8
   Scree=@MaterialExpressionConstant_8
   Scree_Grass=@MaterialExpressionConstant_8
   Heather=@MaterialExpressionConstant_8
   SwampGrass=@MaterialExpressionReroute_33
   SwampReeds=@MaterialExpressionReroute_32
   Forest_Grass=@MaterialExpressionReroute_21
   Beach=@MaterialExpressionConstant_8
   Ocean=@MaterialExpressionConstant_8
   Nightmare-Grass=@MaterialExpressionConstant_8
   Nightmare-Leaves=@MaterialExpressionConstant_8
   Corrupted2=@MaterialExpressionConstant_8
  ) {
   desc="This must have the exact inputs in every biome material!\r\nThis data is baked into the level data when the level is\r\nsaved, so any changes to this require the affected level\r\nto be saved && submitted."
   grassTypes [
    Grass_Standard {
     grassType=/Game/Athena/Apollo/Environments/Landscape/SmallFoliage/DataAssets/LandscapeGrassType/LandscapeGrassType_StandardGrass
     input {
      expression=&self:MaterialExpressionReroute_22
      inputName=None
      mask=0
      maskA=0
      maskB=0
      maskG=0
      maskR=0
      outputIndex=0
     }
    }
    Bush {
     grassType=/Game/Athena/Environments/Landscape/Foliage/Athena_Foliage_Bushes
     input {
      expression=&self:MaterialExpressionReroute_23
      inputName=None
      mask=0
      maskA=0
      maskB=0
      maskG=0
      maskR=0
      outputIndex=0
     }
    }
    FarmFlowers {
     grassType=/Game/Athena/Apollo/Environments/Landscape/SmallFoliage/DataAssets/LandscapeGrassType/LandscapeGrassType_FarmFlowers
     input {
      expression=&self:MaterialExpressionConstant_8
      inputName=None
      mask=0
      maskA=0
      maskB=0
      maskG=0
      maskR=0
      outputIndex=0
     }
    }
    Scree {
     grassType=/Game/Athena/Apollo/Environments/Landscape/SmallFoliage/DataAssets/LandscapeGrassType/LandscapeGrassType_Scree
     input {
      expression=&self:MaterialExpressionConstant_8
      inputName=None
      mask=0
      maskA=0
      maskB=0
      maskG=0
      maskR=0
      outputIndex=0
     }
    }
    Scree_Grass {
     grassType=/Game/Athena/Apollo/Environments/Landscape/SmallFoliage/DataAssets/LandscapeGrassType/LandscapeGrassType_ScreeGrass
     input {
      expression=&self:MaterialExpressionConstant_8
      inputName=None
      mask=0
      maskA=0
      maskB=0
      maskG=0
      maskR=0
      outputIndex=0
     }
    }
    Heather {
     grassType=/Game/Athena/Apollo/Environments/Landscape/SmallFoliage/DataAssets/LandscapeGrassType/LandscapeGrassType_Heather
     input {
      expression=&self:MaterialExpressionConstant_8
      inputName=None
      mask=0
      maskA=0
      maskB=0
      maskG=0
      maskR=0
      outputIndex=0
     }
    }
    SwampGrass {
     grassType=/Game/Athena/Apollo/Environments/Landscape/SmallFoliage/DataAssets/LandscapeGrassType/LandscapeGrassType_SwampGrass
     input {
      expression=&self:MaterialExpressionReroute_33
      inputName=None
      mask=0
      maskA=0
      maskB=0
      maskG=0
      maskR=0
      outputIndex=0
     }
    }
    SwampReeds {
     grassType=/Game/Athena/Apollo/Environments/Landscape/SmallFoliage/DataAssets/LandscapeGrassType/LandscapeGrassType_SwampReeds
     input {
      expression=&self:MaterialExpressionReroute_32
      inputName=None
      mask=0
      maskA=0
      maskB=0
      maskG=0
      maskR=0
      outputIndex=0
     }
    }
    Forest_Grass {
     grassType=/Game/Athena/Apollo/Environments/Landscape/SmallFoliage/DataAssets/LandscapeGrassType/LandscapeGrassType_ForestGrass
     input {
      expression=&self:MaterialExpressionReroute_21
      inputName=None
      mask=0
      maskA=0
      maskB=0
      maskG=0
      maskR=0
      outputIndex=0
     }
    }
    Beach {
     grassType=/Game/Athena/Apollo/Environments/Landscape/SmallFoliage/DataAssets/LandscapeGrassType/LandscapeGrassType_Beach
     input {
      expression=&self:MaterialExpressionConstant_8
      inputName=None
      mask=0
      maskA=0
      maskB=0
      maskG=0
      maskR=0
      outputIndex=0
     }
    }
    Ocean {
     grassType=/Game/Athena/Apollo/Environments/Landscape/SmallFoliage/DataAssets/LandscapeGrassType/LandscapeGrassType_Ocean
     input {
      expression=&self:MaterialExpressionConstant_8
      inputName=None
      mask=0
      maskA=0
      maskB=0
      maskG=0
      maskR=0
      outputIndex=0
     }
    }
    Nightmare-Grass {
     grassType=/Game/Athena/Apollo/Environments/Landscape/SmallFoliage/DataAssets/LandscapeGrassType/LandscapeGrassType_Nightmare
     input {
      expression=&self:MaterialExpressionConstant_8
      inputName=None
      mask=0
      maskA=0
      maskB=0
      maskG=0
      maskR=0
      outputIndex=0
     }
    }
    Nightmare-Leaves {
     grassType=/Game/Athena/Apollo/Environments/Landscape/SmallFoliage/DataAssets/LandscapeGrassType/LandscapeGrassType_Nightmare-Leaves
     input {
      expression=&self:MaterialExpressionConstant_8
      inputName=None
      mask=0
      maskA=0
      maskB=0
      maskG=0
      maskR=0
      outputIndex=0
     }
    }
    Corrupted2 {
     grassType=/Game/Athena/Apollo/Environments/Landscape/SmallFoliage/DataAssets/LandscapeGrassType/LandscapeGrassType_Corrupted2
     input {
      expression=&self:MaterialExpressionConstant_8
      inputName=None
      mask=0
      maskA=0
      maskB=0
      maskG=0
      maskR=0
      outputIndex=0
     }
    }
   ]
  }
  n22 = NamedRerouteUsage()
  n23 = OneMinus(None=@n22) {desc="White=Cliffs"}
  n24 = LandscapeLayerSample() {parameterName=Mud}
  n25 = LandscapeLayerSample() {parameterName=Gravel}
  n26 = Saturate(None=Multiply(A=Max(A=@n24, B=@n25), B=Constant() {r=2}))
  n27 = Subtract(A=@n26, B=@n23) {desc="Remove Cliff from Gravel,Mud && Sand"}
  n28 = LandscapeLayerSample() {parameterName=Sand}
  n29 = Saturate(None=Multiply(A=LandscapeLayerSample() {parameterName=Asphalt}, B=Constant() {r=3}))
  n30 = Max(A=Max(A=@n6.Alpha, B=@n1), B=@n29)
  MaterialExpressionSaturate_12 = Saturate(None=Subtract(A=Saturate(None=@n27), B=Max(A=@n28, B=@n30)) {desc="Remove BiomeB, BiomeC && Asphalt"}) {desc=Final Dirt mask}
  MaterialExpressionReroute_25 = Reroute(None=@n23) {desc=Final Stone mask}
  n31 = Subtract(A=@n22, B=Max(A=@n28, B=@n26)) {desc="Remove Gravel,Mud && Sand"}
  MaterialExpressionSaturate_14 = Saturate(None=Subtract(A=Saturate(None=@n31), B=@n29) {desc=Remove Asphalt}) {desc=Final Grass mask}
  MaterialExpressionReroute_37 = Reroute(None=@n29)
  n32 = Saturate(None=Subtract(A=@n28, B=@n23) {desc=Remove Cliff})
  MaterialExpressionReroute_39 = Reroute(None=Saturate(None=Subtract(A=@n32, B=@n30) {desc="Remove ForestFloor, Asphalt, SwampGrass"}))
  LandscapePhysicalMaterialOutput!(
   Dirt=@MaterialExpressionSaturate_12
   Stone=@MaterialExpressionReroute_25
   Grass=@MaterialExpressionSaturate_14
   Asphalt=@MaterialExpressionReroute_37
   Sand=@MaterialExpressionReroute_39
  ) {
   desc="These are layered in order of importance.\r\nIf Dirt=1 and Stone=1, then Dirt will \"win\".\r\nTheoretically, these values aren't clamped.\r\nSo a value of 2 should trump a value of 1."
   inputs [
    {
     physicalMaterial=/Game/FortressPhysicalMaterials/Dirt
     input {
      expression=&self:MaterialExpressionSaturate_12
      inputName=None
      mask=0
      maskA=0
      maskB=0
      maskG=0
      maskR=0
      outputIndex=0
     }
    }
    {
     physicalMaterial=/Game/FortressPhysicalMaterials/Stone
     input {
      expression=&self:MaterialExpressionReroute_25
      inputName=None
      mask=0
      maskA=0
      maskB=0
      maskG=0
      maskR=0
      outputIndex=0
     }
    }
    {
     physicalMaterial=/Game/FortressPhysicalMaterials/Grass
     input {
      expression=&self:MaterialExpressionSaturate_14
      inputName=None
      mask=0
      maskA=0
      maskB=0
      maskG=0
      maskR=0
      outputIndex=0
     }
    }
    {
     physicalMaterial=/Game/FortressPhysicalMaterials/Asphalt
     input {
      expression=&self:MaterialExpressionReroute_37
      inputName=None
      mask=0
      maskA=0
      maskB=0
      maskG=0
      maskR=0
      outputIndex=0
     }
    }
    {
     physicalMaterial=/Game/FortressPhysicalMaterials/Sand
     input {
      expression=&self:MaterialExpressionReroute_39
      inputName=None
      mask=0
      maskA=0
      maskB=0
      maskG=0
      maskR=0
      outputIndex=0
     }
    }
   ]
  }
  n33 = LinearInterpolate(A=@MaterialExpressionSaturate_12, B=@MaterialExpressionReroute_25, Alpha=ScalarParameter() {parameterName=DEBUG-PhysMat1})
  n34 = ScalarParameter() {parameterName=DEBUG-PhysMat2}
  LinearInterpolate(
   A=LinearInterpolate(
    A=LinearInterpolate(A=@n33, B=@MaterialExpressionSaturate_14, Alpha=@n34)
    B=@MaterialExpressionReroute_37
    Alpha=ScalarParameter() {parameterName=DEBUG-PhysMat3}
   )
   B=@MaterialExpressionReroute_39
   Alpha=ScalarParameter() {defaultValue=1, parameterName=DEBUG-PhysMat4}
  )
  n35 = MaterialFunctionCall!(
   AdditionalGrassBorderColorMask=Max(A=LandscapeLayerSample() {desc=ForestFloor, parameterName=BiomeB}, B=LandscapeLayerSample() {desc=WetMud, parameterName=BiomeA})
   PutSandBeforeGravel
   UseOceanLevel
  ) {materialFunction=/Game/Athena/Apollo/Environments/Landscape/Materials/Biome/CommonMaterials/MF_CommonLandscapeLayers}
  n36 = MaterialFunctionCall!(
   Diffuse=@n35.GravelDiffuse
   Normals=@n35.GravelDiffuse
   Height=@n35.GravelDiffuse
   BiomeColorDiv=VectorParameter() {parameterName=Biome-WetMudDiv, defaultValue=<rgba 0.85 0.6208 0.425 1>}.RGB
   BiomeColor=VectorParameter() {parameterName=Biome-WetMudColor, defaultValue=<rgba 0.2 0.19 0.16 1>}.RGB
   BiomeColorDark=VectorParameter() {parameterName=Biome-WetMudColorDark, defaultValue=<rgba 0.05 0.0477 0.045 1>}.RGB
  ) {materialFunction=/Game/Athena/Apollo/Environments/Landscape/Materials/Biome/UniqueBiomeMaterialFunctions/MF_Biome_WetMud}
  n37 = MaterialFunctionCall!(
   Material 1=@n35.Material
   Height Texture 1=@n35.Material
   Material 2=@n36.Material
   Height Texture 2=@n36.Material
   Transistion Phase=LandscapeLayerSample() {desc=WetMud, parameterName=BiomeA}
   UseContrast
   Contrast
   AddLayerMask
  ) {materialFunction=/Game/Athena/Apollo/Environments/Landscape/Materials/Biome/CommonMaterials/MF_BlendLandscapeLayers}
  n38 = MaterialFunctionCall() {materialFunction=/Game/Athena/Apollo/Environments/Landscape/Materials/Biome/UniqueBiomeMaterialFunctions/MF_Biome_Forest}
  n39 = MaterialFunctionCall(
   Material 1=@n37.BlendedMaterial
   Height Texture 1=@n37.BlendedMaterial
   Material 2=@n38.Material
   Height Texture 2=NamedRerouteDeclaration(None=@n38.Height) {name=MF_Biome_Forest-Height, nodeColor=<rgba 0.5 0 1 1>}
   Transistion Phase=LandscapeLayerSample() {desc=ForestFloor, parameterName=BiomeB}
  ) {materialFunction=&MF_BlendLandscapeLayers}
  n40 = MaterialFunctionCall() {materialFunction=/Game/Athena/Apollo/Environments/Landscape/Materials/Biome/UniqueBiomeMaterialFunctions/MF_Biome_SwampGrass}
  n41 = MaterialFunctionCall(
   Material 1=@n39.BlendedMaterial
   Height Texture 1=@n39.BlendedMaterial
   Material 2=@n40.Material
   Height Texture 2=@n40.Material
   Transistion Phase=LandscapeLayerSample() {parameterName=BiomeC}
  ) {materialFunction=&MF_BlendLandscapeLayers}
  n42 = MaterialFunctionCall!(Material=@n41.BlendedMaterial, LandscapeLayer=LandscapeLayerSample() {parameterName=BiomeD}) {materialFunction=/Game/Athena/Apollo/Environments/Landscape/Materials/Biome/UniqueBiomeMaterialFunctions/MF_Biome_Autumn_Leaves}
  MaterialFunctionCall(
   Material 1=@n41.BlendedMaterial
   Height Texture 1=@n41.BlendedMaterial
   Material 2=@n42.Material
   Height Texture 2=@n42.Material
   Transistion Phase=@n42.Material
   UseContrast=StaticBool() {value=true}
   Contrast=@n42.Material
  ) {materialFunction=&MF_BlendLandscapeLayers}
  n43 = GetMaterialAttributes(None=MaterialFunctionCall!(PutSandBeforeGravelOverride, UseSnow=StaticBool(), UseOceanLevel) {materialFunction=/Game/Athena/Apollo/Environments/Landscape/Materials/Biome/CommonMaterials/Mobile/MF_Mobile_CommonLandscapeLayers}.Material) {attributeGetTypes [#1]}
  MaterialExpressionConstant_10 = Constant()
  n44 = LandscapeLayerBlend!(
   Layer Gravel=@MaterialExpressionConstant_10
   Layer Mud=@MaterialExpressionConstant_10
   Layer Sand=@MaterialExpressionConstant_10
   Layer Asphalt=@MaterialExpressionConstant_10
  ) {layers [
    {
     blendType=LB_WeightBlend
     constHeightInput=0
     layerName=Gravel
     previewWeight=0
     constLayerInput=<0 0 0>
     heightInput {expression=null, inputName=None, mask=0, maskA=0, maskB=0, maskG=0, maskR=0, outputIndex=0}
     layerInput {
      expression=&self:MaterialExpressionConstant_10
      inputName=None
      mask=0
      maskA=0
      maskB=0
      maskG=0
      maskR=0
      outputIndex=0
     }
    }
    {
     blendType=LB_WeightBlend
     constHeightInput=0
     layerName=Mud
     previewWeight=0
     constLayerInput=<0 0 0>
     heightInput {expression=null, inputName=None, mask=0, maskA=0, maskB=0, maskG=0, maskR=0, outputIndex=0}
     layerInput {
      expression=&self:MaterialExpressionConstant_10
      inputName=None
      mask=0
      maskA=0
      maskB=0
      maskG=0
      maskR=0
      outputIndex=0
     }
    }
    {
     blendType=LB_WeightBlend
     constHeightInput=0
     layerName=Sand
     previewWeight=0
     constLayerInput=<0 0 0>
     heightInput {expression=null, inputName=None, mask=0, maskA=0, maskB=0, maskG=0, maskR=0, outputIndex=0}
     layerInput {
      expression=&self:MaterialExpressionConstant_10
      inputName=None
      mask=0
      maskA=0
      maskB=0
      maskG=0
      maskR=0
      outputIndex=0
     }
    }
    {
     blendType=LB_WeightBlend
     constHeightInput=0
     layerName=Asphalt
     previewWeight=0
     constLayerInput=<0 0 0>
     heightInput {expression=null, inputName=None, mask=0, maskA=0, maskB=0, maskG=0, maskR=0, outputIndex=0}
     layerInput {
      expression=&self:MaterialExpressionConstant_10
      inputName=None
      mask=0
      maskA=0
      maskB=0
      maskG=0
      maskR=0
      outputIndex=0
     }
    }
   ]}
  n45 = SetMaterialAttributes!(MaterialAttributes=@n43.MaterialAttributes, Base Color=Add(A=@n43.Base Color, B=Multiply(A=@n44) {constB=0})) {desc="This must remain in material,\r\nnot a function.", attributeSetTypes [#1]}
  n46 = RuntimeVirtualTextureSampleParameter() {
   group=RVT
   materialType=BaseColor_Normal_Specular_YCoCg
   parameterName=RVT Sample Target
   virtualTexture=/Game/Athena/Apollo/Maps/Landscape/VT/RVT_Apollo_Landscape
  }
  n47 = SetMaterialAttributes!(MaterialAttributes, Base Color=@n46.BaseColor, Specular=@n46.BaseColor, Roughness=@n46.BaseColor, Normal=@n46.BaseColor) {attributeSetTypes [
    #1
    #2
    #3
    #4
   ]}
  n48 = MaterialFunctionCall!(
   Material=@n41.BlendedMaterial
   Heightmap=@n41.BlendedMaterial
   NewSlopeMaskMask
   Displacement
   AsphaltLayer=LandscapeLayerSample() {parameterName=Sand}
   UseAlternateCliffMat
   AltCliffHeight
   AltCliffMat
  ) {materialFunction=/Game/Athena/Apollo/Environments/Landscape/Materials/Biome/CommonMaterials/MF_SlopeBlendCliff}
  n49 = MaterialFunctionCall!(
   MobileMaterial=@n45
   HighEndMaterial=MaterialFunctionCall!(Material=@n48.Material) {materialFunction=/Game/Athena/Apollo/Maps/Landscape/VT/MaterialFunctions/TransformTangentToWorld}.Result
  ) {materialFunction=/Game/Athena/Asteria/Environments/Landscape/Materials/MaterialFunctions/FN_RvtWriteSwitch}
  n50 = MaterialFunctionCall(
   NoRVT=@n45
   RVT=StaticSwitchParameter(
    True=MaterialFunctionCall!(RVT=@n47, ByPass RVT=@n49.Result) {materialFunction=/Game/Athena/Apollo/Maps/Landscape/VT/MaterialFunctions/EnableRVT}.Result
    False=@n49.Result
   ) {defaultValue=true, group=RVT Base, parameterName=RVT Enable}
  ) {materialFunction=&FN_RVT_Switch_Material}
  n51 = SetMaterialAttributes!(MaterialAttributes=@n50.Result, Opacity Mask=LandscapeVisibilityMask()) {attributeSetTypes [#5]}
  MaterialFunctionCall!(
   In=ShadingPathSwitch(Default=@n51, Deferred=@n51, Forward=@n51, Mobile=@n50.Result)
   UseHeightFalloff
   OnFireEmissiveOverride
   HeightFalloffRadius
   HeightFalloffHardness
   VariationMasksMobile
   In_Grass
   Is_Stylised_WE
  ) {desc=This should be very last, materialFunction=/Game/Packages/Fortress_SharedMaterials/Base_Material_Functions/MF_Curie_Landscape}
  n52 = Saturate(None=Multiply(A=Constant() {r=2}, B=LandscapeLayerSample() {parameterName=BiomeD}))
  n53 = OneMinus(None=@n52)
  Multiply(A=@n53, B=@MaterialExpressionReroute_23)
  Multiply(A=@n53, B=@MaterialExpressionReroute_22)
  Multiply(A=@n52, B=@MaterialExpressionReroute_22)
  NamedRerouteDeclaration(None=@n35.GravelMaskForGrassFoliage) {name=MaskForGrassFoliage, nodeColor=<rgba 0.5 0 1 1>}
  NamedRerouteDeclaration(None=MaterialFunctionCall!(UseLandscapeCoords=StaticBool() {value=true}) {materialFunction=/Game/Athena/Apollo/Environments/Landscape/Materials/Biome/CommonMaterials/MF_ApolloTerrainMasks}.NotInUse) {name=ForestBiomeMask, nodeColor=<rgba 0.5 0 1 1>}
  NamedRerouteDeclaration(None=@n48.HeightmapForGrass) {name=HeightmapForGrass, nodeColor=<rgba 1 1 0 1>}
  NamedRerouteDeclaration(None=@n48.CliffMask) {name=CliffMask, nodeColor=<rgba 0.5 0 1 1>}
  NamedRerouteDeclaration(None=@n48.LandscapeGrassMeshMask) {name=GrassMeshMask, nodeColor=<rgba 0.5 0 1 1>}
  n54 = GetMaterialAttributes(None=@n49.Result) {attributeGetTypes [
    #1
    #2
    #3
    #4
   ]}
  RuntimeVirtualTextureOutput!(
   BaseColor=MaterialFunctionCall!(BaseColor=@n54.Base Color) {materialFunction=/Game/Athena/Apollo/Environments/Landscape/Materials/Biome/CommonMaterials/MF_ReduceRTV-CompressionArtifacts}.BaseColor
   Specular=@n54.Specular
   Roughness=@n54.Specular
   Normal=@n54.Specular
   WorldHeight=ComponentMask(None=WorldPosition().XYZ) {b=true}
   Opacity
   Mask
   Displacement
   Mask4
  )
 ]
}
