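# Landscape material for the Apollo lobby island, with Runtime Virtual Texture read and write and a separate mobile layer path.
# Common layers blend with Forest and FarmGrass biomes by painted weights, then slope-based cliffs, feeding the virtual texture.
# Layer masks drive grass spawning and Dirt/Stone physical materials; Beach and height-banded Ocean grass are zeroed on forward and mobile.
# BiomeA tint and autumn-leaf blends feed nothing.
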Material /Game/Athena/Apollo/Environments/Landscape/Materials/Biome/LandscapeBiomeMaterials/M_Lobby-ONLY
graph {
 properties {bTangentSpaceNormal=false, bUseMaterialAttributes=true, pixelDepthOffsetMode=PDOM_Legacy, usedShadingModels=Default Lit}
 nodes [
  n1 = LandscapeLayerSample() {parameterName=Gravel}
  n2 = LandscapeLayerSample() {parameterName=Mud}
  n3 = Max(A=Multiply(A=Max(A=@n1, B=@n2), B=Constant() {r=2}), B=Multiply(A=LandscapeLayerSample() {parameterName=Sand}, B=Constant() {r=5}))
  n4 = Ceil(None=LandscapeLayerSample() {parameterName=Asphalt})
  n5 = NamedRerouteUsage()
  n6 = Max(A=@n5, B=Saturate(None=Max(A=@n3, B=@n4)))
  MaterialExpressionReroute_1 = Reroute(None=OneMinus(None=@n6))
  MaterialExpressionAdd_2 = Add(A=Constant() {r=-0.5}, B=Multiply(A=Constant() {r=2}, B=LandscapeLayerSample() {parameterName=BiomeC}))
  n7 = Constant()
  n8 = Subtract(A=LandscapeLayerSample() {parameterName=Sand}, B=@n5)
  MaterialExpressionShadingPathSwitch_3 = ShadingPathSwitch(Default=@n8, Deferred=@n8, Forward=@n7, Mobile=@n7)
  n9 = Constant()
  n10 = ComponentMask(None=WorldPosition().XYZ) {b=true}
  n11 = Constant() {r=-4400}
  n12 = Constant() {r=1}
  n13 = Saturate(None=Divide(A=Subtract(A=@n10, B=@n11), B=@n12))
  n14 = Constant() {r=-5100}
  n15 = Saturate(None=Divide(A=Subtract(A=@n10, B=@n14), B=@n12))
  n16 = Subtract(A=Subtract(A=OneMinus(None=@n13), B=OneMinus(None=@n15)), B=@n5)
  MaterialExpressionShadingPathSwitch_2 = ShadingPathSwitch(Default=@n16, Deferred=@n16, Forward=@n9, Mobile=@n9)
  LandscapeGrassOutput!(
   Grass_Standard=@MaterialExpressionReroute_1
   FarmFlowers=@MaterialExpressionAdd_2
   Beach=@MaterialExpressionShadingPathSwitch_3
   Ocean=@MaterialExpressionShadingPathSwitch_2
  ) {grassTypes [
    Grass_Standard {
     grassType=/Game/Athena/Apollo/Environments/Landscape/SmallFoliage/DataAssets/LandscapeGrassType/LandscapeGrassType_StandardGrass_Lobby
     input {
      expression=&self:MaterialExpressionReroute_1
      inputName=None
      mask=0
      maskA=0
      maskB=0
      maskG=0
      maskR=0
      outputIndex=0
     }
    }
    FarmFlowers {
     grassType=/Game/Athena/Apollo/Environments/Landscape/SmallFoliage/DataAssets/LandscapeGrassType/LandscapeGrassType_FarmFlowers
     input {
      expression=&self:MaterialExpressionAdd_2
      inputName=None
      mask=0
      maskA=0
      maskB=0
      maskG=0
      maskR=0
      outputIndex=0
     }
    }
    Beach {
     grassType=/Game/Athena/Apollo/Environments/Landscape/SmallFoliage/DataAssets/LandscapeGrassType/LandscapeGrassType_Beach
     input {
      expression=&self:MaterialExpressionShadingPathSwitch_3
      inputName=None
      mask=0
      maskA=0
      maskB=0
      maskG=0
      maskR=0
      outputIndex=0
     }
    }
    Ocean {
     grassType=/Game/Athena/Apollo/Environments/Landscape/SmallFoliage/DataAssets/LandscapeGrassType/LandscapeGrassType_Ocean
     input {
      expression=&self:MaterialExpressionShadingPathSwitch_2
      inputName=None
      mask=0
      maskA=0
      maskB=0
      maskG=0
      maskR=0
      outputIndex=0
     }
    }
   ]}
  MaterialExpressionReroute_0 = Reroute(None=@n6)
  MaterialExpressionOneMinus_0 = OneMinus(None=NamedRerouteUsage())
  LandscapePhysicalMaterialOutput!(Dirt=@MaterialExpressionReroute_0, Stone=@MaterialExpressionOneMinus_0) {inputs [
    {
     physicalMaterial=/Game/FortressPhysicalMaterials/Dirt
     input {
      expression=&self:MaterialExpressionReroute_0
      inputName=None
      mask=0
      maskA=0
      maskB=0
      maskG=0
      maskR=0
      outputIndex=0
     }
    }
    {
     physicalMaterial=/Game/FortressPhysicalMaterials/Stone
     input {
      expression=&self:MaterialExpressionOneMinus_0
      inputName=None
      mask=0
      maskA=0
      maskB=0
      maskG=0
      maskR=0
      outputIndex=0
     }
    }
   ]}
  n17 = GetMaterialAttributes(None=MaterialFunctionCall!(PutSandBeforeGravelOverride, UseSnow=StaticBool(), UseOceanLevel) {materialFunction=/Game/Athena/Apollo/Environments/Landscape/Materials/Biome/CommonMaterials/Mobile/MF_Mobile_CommonLandscapeLayers}.Material) {attributeGetTypes [#1]}
  MaterialExpressionConstant_10 = Constant()
  n18 = LandscapeLayerBlend!(
   Layer Gravel=@MaterialExpressionConstant_10
   Layer Mud=@MaterialExpressionConstant_10
   Layer Sand=@MaterialExpressionConstant_10
   Layer Asphalt=@MaterialExpressionConstant_10
  ) {layers [
    {
     blendType=LB_WeightBlend
     constHeightInput=0
     layerName=Gravel
     previewWeight=0
     constLayerInput=<0 0 0>
     heightInput {expression=null, inputName=None, mask=0, maskA=0, maskB=0, maskG=0, maskR=0, outputIndex=0}
     layerInput {
      expression=&self:MaterialExpressionConstant_10
      inputName=None
      mask=0
      maskA=0
      maskB=0
      maskG=0
      maskR=0
      outputIndex=0
     }
    }
    {
     blendType=LB_WeightBlend
     constHeightInput=0
     layerName=Mud
     previewWeight=0
     constLayerInput=<0 0 0>
     heightInput {expression=null, inputName=None, mask=0, maskA=0, maskB=0, maskG=0, maskR=0, outputIndex=0}
     layerInput {
      expression=&self:MaterialExpressionConstant_10
      inputName=None
      mask=0
      maskA=0
      maskB=0
      maskG=0
      maskR=0
      outputIndex=0
     }
    }
    {
     blendType=LB_WeightBlend
     constHeightInput=0
     layerName=Sand
     previewWeight=0
     constLayerInput=<0 0 0>
     heightInput {expression=null, inputName=None, mask=0, maskA=0, maskB=0, maskG=0, maskR=0, outputIndex=0}
     layerInput {
      expression=&self:MaterialExpressionConstant_10
      inputName=None
      mask=0
      maskA=0
      maskB=0
      maskG=0
      maskR=0
      outputIndex=0
     }
    }
    {
     blendType=LB_WeightBlend
     constHeightInput=0
     layerName=Asphalt
     previewWeight=0
     constLayerInput=<0 0 0>
     heightInput {expression=null, inputName=None, mask=0, maskA=0, maskB=0, maskG=0, maskR=0, outputIndex=0}
     layerInput {
      expression=&self:MaterialExpressionConstant_10
      inputName=None
      mask=0
      maskA=0
      maskB=0
      maskG=0
      maskR=0
      outputIndex=0
     }
    }
   ]}
  n19 = SetMaterialAttributes!(MaterialAttributes=@n17.MaterialAttributes, Base Color=Add(A=@n17.Base Color, B=Multiply(A=@n18) {constB=0})) {desc="This must remain in material,\r\nnot a function.", attributeSetTypes [#1]}
  n20 = RuntimeVirtualTextureSampleParameter() {
   group=RVT
   materialType=BaseColor_Normal_Specular_YCoCg
   parameterName=RVT Sample Target
   virtualTexture=/Game/Athena/Apollo/Maps/Landscape/VT/RVT_Apollo_LobbyIsland
  }
  n21 = SetMaterialAttributes!(MaterialAttributes, Base Color=@n20.BaseColor, Specular=@n20.BaseColor, Roughness=@n20.BaseColor, Normal=@n20.BaseColor) {attributeSetTypes [
    #1
    #2
    #3
    #4
   ]}
  n22 = MaterialFunctionCall!(AdditionalGrassBorderColorMask, PutSandBeforeGravel, UseOceanLevel) {materialFunction=/Game/Athena/Apollo/Environments/Landscape/Materials/Biome/CommonMaterials/MF_CommonLandscapeLayers}
  n23 = MaterialFunctionCall() {materialFunction=/Game/Athena/Apollo/Environments/Landscape/Materials/Biome/UniqueBiomeMaterialFunctions/MF_Biome_Forest}
  n24 = MaterialFunctionCall!(
   Material 1=@n22.Material
   Height Texture 1=@n22.Material
   Material 2=@n23.Material
   Height Texture 2=@n23.Material
   Transistion Phase=LandscapeLayerSample() {parameterName=BiomeB}
   UseContrast
   Contrast
   AddLayerMask
  ) {materialFunction=/Game/Athena/Apollo/Environments/Landscape/Materials/Biome/CommonMaterials/MF_BlendLandscapeLayers}
  n25 = MaterialFunctionCall!(
   GrassHeightmap=NamedRerouteUsage()
   GrassNoiseMask=NamedRerouteUsage()
   GrassBaseColorDetail=NamedRerouteUsage()
   GrassMaterial=NamedRerouteUsage()
  ) {materialFunction=/Game/Athena/Apollo/Environments/Landscape/Materials/Biome/UniqueBiomeMaterialFunctions/MF_Biome_FarmGrass}
  n26 = MaterialFunctionCall(
   Material 1=@n24.BlendedMaterial
   Height Texture 1=@n24.BlendedMaterial
   Material 2=@n25.Material
   Height Texture 2=@n25.Material
   Transistion Phase=LandscapeLayerSample() {parameterName=BiomeC}
  ) {materialFunction=&MF_BlendLandscapeLayers}
  n27 = MaterialFunctionCall!(
   Material=@n26.BlendedMaterial
   Heightmap=@n26.BlendedMaterial
   NewSlopeMaskMask
   Displacement
   AsphaltLayer
   UseAlternateCliffMat
   AltCliffHeight
   AltCliffMat
  ) {materialFunction=/Game/Athena/Apollo/Environments/Landscape/Materials/Biome/CommonMaterials/MF_SlopeBlendCliff}
  n28 = MaterialFunctionCall!(
   MobileMaterial=@n19
   HighEndMaterial=MaterialFunctionCall!(Material=@n27.Material) {materialFunction=/Game/Athena/Apollo/Maps/Landscape/VT/MaterialFunctions/TransformTangentToWorld}.Result
  ) {materialFunction=/Game/Athena/Asteria/Environments/Landscape/Materials/MaterialFunctions/FN_RvtWriteSwitch}
  MaterialFunctionCall(
   NoRVT=@n19
   RVT=StaticSwitchParameter(
    True=MaterialFunctionCall!(RVT=@n21, ByPass RVT=@n28.Result) {materialFunction=/Game/Athena/Apollo/Maps/Landscape/VT/MaterialFunctions/EnableRVT}.Result
    False=@n28.Result
   ) {defaultValue=true, group=RVT Base, parameterName=RVT Enable}
  ) {materialFunction=&FN_RVT_Switch_Material}
  n29 = MaterialFunctionCall!(Material=@n26.BlendedMaterial, LandscapeLayer=LandscapeLayerSample() {parameterName=BiomeD}) {materialFunction=/Game/Athena/Apollo/Environments/Landscape/Materials/Biome/UniqueBiomeMaterialFunctions/MF_Biome_Autumn_Leaves}
  MaterialFunctionCall(
   Material 1=@n26.BlendedMaterial
   Height Texture 1=@n26.BlendedMaterial
   Material 2=@n29.Material
   Height Texture 2=@n29.Material
   Transistion Phase=@n29.Material
   UseContrast=StaticBool() {value=true}
   Contrast=@n29.Material
  ) {materialFunction=&MF_BlendLandscapeLayers}
  n30 = GetMaterialAttributes(None=@n22.Material) {attributeGetTypes [#1]}
  n31 = VectorParameter() {parameterName=BiomeA-TestColor, defaultValue=<rgba 1.1 0.4533 0 1>}
  MaterialFunctionCall(
   Material 1=@n22.Material
   Height Texture 1=@n22.Heightmap
   Material 2=SetMaterialAttributes!(MaterialAttributes=@n30.MaterialAttributes, Base Color=Multiply(A=@n30.Base Color, B=@n31.RGB)) {attributeSetTypes [#1]}
   Height Texture 2=@n22.Heightmap
   Transistion Phase=LandscapeLayerSample() {parameterName=BiomeA}
  ) {materialFunction=&MF_BlendLandscapeLayers}
  n32 = Constant() {r=2}
  n33 = LandscapeLayerSample() {parameterName=BiomeD}
  Multiply(A=OneMinus(None=Multiply(A=@n32, B=@n33)), B=@MaterialExpressionReroute_1)
  NamedRerouteDeclaration(None=@n22.GrassNoiseMask) {name=GrassNoiseMask, nodeColor=<rgba 0.5 0 1 1>}
  NamedRerouteDeclaration(None=@n22.GrassBaseColorDetail) {name=GrassBaseColorDetail, nodeColor=<rgba 0.5 0 1 1>}
  NamedRerouteDeclaration(None=@n22.Material) {name=CommonMaterial, nodeColor=<rgba 0.5 0 1 1>}
  NamedRerouteDeclaration(None=@n22.GrassHeightmap) {name=GrassHeightmap, nodeColor=<rgba 0.5 0 1 1>}
  NamedRerouteDeclaration(None=@n27.CliffMask) {name=CliffMask, nodeColor=<rgba 0.5 0 1 1>}
  NamedRerouteDeclaration(None=@n27.LandscapeGrassMeshMask) {name=GrassMeshMask, nodeColor=<rgba 0.5 0 1 1>}
  n34 = GetMaterialAttributes(None=@n28.Result) {attributeGetTypes [
    #1
    #2
    #3
    #4
   ]}
  RuntimeVirtualTextureOutput!(
   BaseColor=MaterialFunctionCall!(BaseColor=@n34.Base Color) {materialFunction=/Game/Athena/Apollo/Environments/Landscape/Materials/Biome/CommonMaterials/MF_ReduceRTV-CompressionArtifacts}.BaseColor
   Specular=@n34.Specular
   Roughness=@n34.Specular
   Normal=@n34.Specular
   WorldHeight=ComponentMask(None=WorldPosition().XYZ) {b=true}
   Opacity
   Mask
   Displacement
   Mask4
  )
 ]
}
